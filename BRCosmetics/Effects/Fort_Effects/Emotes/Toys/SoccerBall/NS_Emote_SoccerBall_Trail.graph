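# Niagara system for the trail of the soccer ball toy emote, built from a single Body emitter with ribbon and sprite renderers.
# Particles spawn per unit of distance moved, so the trail follows motion; the rate-based spawn is disabled.
# Each short-lived white particle drifts under slight random horizontal acceleration, narrows in ribbon width, and follows an alpha fade curve over its life.
# Dynamic material parameters are set once at spawn.
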NiagaraSystem /BRCosmetics/Effects/Fort_Effects/Emotes/Toys/SoccerBall/NS_Emote_SoccerBall_Trail
dependencies {usedDataInterfaces [&NiagaraDataInterfaceCurve], usedDynamicInputs [], usedModules [&SystemState]}
emitters {Body {
  emitterUpdateScript [
   &EmitterState(Inactive Response=::NewEnumerator1, Life Cycle Mode=::NewEnumerator1, Loop Duration=1, UseLoopDelay=false)
   &SpawnPerUnit(Module.Use Max Movement Threshold=true, Module.Use Movement Tolerance=false, Spawn Spacing=1, Use Spawn Probability=false)
   &SpawnRate(SpawnRate=60, Use Spawn Probability=false) {enabled=false}
  ]
  particleSpawnScript [
   /SetVariables(Particles.Acceleration:Vector3f=&RandomRangeVector(Minimum=<-3 -3 0>, Maximum=<3 3 0>, Evaluation Type=::NewEnumerator1)) {isSetParameters=true}
   &InitializeRibbon(
    Lifetime=0.2
    Module.Write Color=true
    Module.Write Lifetime=true
    Module.Write Mass=true
    Module.Write Material Random=true
    Module.Write Position=false
    Module.Write RibbonTwist=false
    Module.Write RibbonWidth=true
    Ribbon Width=20
   )
   &DynamicMaterialParameters(
    Index 0 Param 1=0.5
    Index 0 Param 2=0.25
    Index 0 Param 3=1
    Index 0 Param 4=1
    Module.Index0Float0Write=true
    Module.Index0Float1Write=true
    Module.Index0Float2Write=true
    Module.Index0Float3Write=true
   )
   &AddVelocity(Velocity=<500 0 0>) {enabled=false}
  ]
  particleUpdateScript [
   /SetVariables(Particles.ScaledRibbonWidth:NiagaraFloat=&Multiply_Float(
     A=$Particles.RibbonWidth
     B=&FloatFromCurve(FloatCurve=DataInterface() {propertyValues {Curve {keys [{tangentMode=RCTM_User, time=0, value=1}, {tangentMode=RCTM_User, time=1, value=0.25}]}}})
    )) {isSetParameters=true}
   &ParticleState()
   &Color(Color=&MakeLinearColorFromVectorAndFloat(
     "Vector (RGB)"=<1 1 1>
     "Float (Alpha)"=&FloatFromCurve(FloatCurve=DataInterface() {propertyValues {Curve {keys [
          <at 0.1 0 RCIM_Cubic>
          {arriveTangent=-0.0173, interpMode=RCIM_Cubic, leaveTangent=-0.0173, tangentMode=RCTM_User, time=0.15, value=0.5}
          {arriveTangent=-0.878, interpMode=RCIM_Cubic, leaveTangent=-0.878, tangentMode=RCTM_User, time=0.4, value=0.2}
          <at 1 0 RCIM_Cubic>
         ]}}})
    ))
   &AccelerationForce(Acceleration=$Particles.Acceleration)
   &SolveForcesAndVelocity(Clamp Velocity=false, Limit Acceleration=false)
  ]
  rendererClasses [&NiagaraRibbonRendererProperties, &NiagaraSpriteRendererProperties]
  renderers [
   {rendererClass=&NiagaraRibbonRendererProperties}
   {rendererClass=&NiagaraSpriteRendererProperties, rendererIndex=1}
  ]
 }}
summary {
 systemName=NS_Emote_SoccerBall_Trail
 emitters [{emitterName=Body, rendererClasses [&NiagaraRibbonRendererProperties, &NiagaraSpriteRendererProperties]}]
 userVariables []
}
